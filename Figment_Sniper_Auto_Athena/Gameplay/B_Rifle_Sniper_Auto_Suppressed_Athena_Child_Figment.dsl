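; Blueprint child of the suppressed automatic sniper rifle weapon for the Figment variant, adjusting its firing effects while scoped.
; It overrides the shell ejection and muzzle flash FX functions, forwarding to the parent only when the scope post-process is not enabled.
; Both effects are thus skipped during scoped view.
; Construction defers to the parent, while BeginPlay holds a disabled parent call and Tick is empty.
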(event EventBeginPlay
  (|CallParentFunction) ; This node is disabled and will not be called.\nDrag off pins to build functionality.
)

(event EventTick
  (params (DeltaSeconds :type Float%20%28single-precision%29)))

(fn ActivateShellsFX
  (params (Bool :type Boolean))
  (if (not (Variables|Scope|GetScopePostProcessEnabled))
    (|CallParentFunction Bool)))

(fn MuzzleFlashFX
  (params (Persistent%20Fire :type Boolean))
  (if (not (Variables|Scope|GetScopePostProcessEnabled))
    (|CallParentFunction Persistent%20Fire)))

(fn ConstructionScript
  (|CallParentFunction))
